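# UI-domain material for Lettuce-themed Item Shop tiles, built around animated hexagon patterns in pinks and reds.
# Custom HLSL hexagon SDFs on time-scaled or offset UVs form expanding hexagons and rippling bands, often quantized by rounding into three-color ramps.
# Linear and radial gradients and a shard texture mask blend in highlights and accent colors.
# The graph holds several parallel variants and loose nodes, with no material output connection shown.
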Material /ItemShop/UI/Art/Lettuce/M_UI_Lettuce_ShopTiles
graph {
 properties {bUsedWithStaticMesh=false, materialDomain=MD_UI, pixelDepthOffsetMode=PDOM_Legacy, usedShadingModels=Default Lit}
 nodes [
  n1 = Time()
  n2 = Frac(None=Multiply(A=@n1, B=ScalarParameter() {defaultValue=0.15, parameterName=speed}))
  n3 = Constant() {r=1}
  MaterialExpressionConstantBiasScale_1 = ConstantBiasScale(None=MaterialFunctionCall(
    UVs=MaterialFunctionCall() {materialFunction=&MF_UI_UVScaling}.Scaled to Fit
    Texture Scale=Multiply(A=@n2, B=@n3)
   ) {materialFunction=&ScaleUVsByCenter}.UVs) {bias=-0.5, scale=2}
  MaterialExpressionConstant_4 = Constant() {r=1}
  n4 = Custom!(p=@MaterialExpressionConstantBiasScale_1, r=@MaterialExpressionConstant_4) {
   code="const float3 k = float3(-0.866025404,0.5,0.577350269);\r\n    p = abs(p);\r\n    p -= 2.0*min(dot(k.xy,p),0.0)*k.xy;\r\n    p -= float2(clamp(p.x, -k.z*r, k.z*r), r);\r\n    return length(p)*sign(p.y);"
   outputType=CMOT_Float1
   inputs [
    {inputName=p, input {expression=&self:MaterialExpressionConstantBiasScale_1}}
    {inputName=r, input {expression=&self:MaterialExpressionConstant_4}}
   ]
  }
  n5 = Multiply(A=@n4, B=Constant() {desc=RippleFrequency, r=1.0074})
  Abs(None=Frac(None=@n5))
  Abs()
  n6 = Multiply(A=Time(), B=Constant() {r=0.009})
  n7 = Multiply(A=Time(), B=Constant() {desc=RippleFrequency, r=0.009})
  Add(A=@n6, B=@n7)
  Add(A=@n1, B=Constant() {r=1}) {constB=0.25}
  Add(B=Constant() {r=0.05}) {constB=0.01}
  n8 = MaterialFunctionCall(X=Constant() {r=-1.2}, Y=Constant() {r=-1.2}) {materialFunction=&MakeFloat2}
  n9 = MaterialFunctionCall() {materialFunction=&MF_UI_UVScaling}
  MaterialExpressionConstantBiasScale_4 = ConstantBiasScale(None=MaterialFunctionCall(UVs=Add(A=@n8.Result, B=@n9.Scaled To Fill), Texture Scale=Constant() {r=0.8}) {materialFunction=&ScaleUVsByCenter}.UVs)
  MaterialExpressionConstant_15 = Constant() {r=1}
  n10 = Custom!(uv=@MaterialExpressionConstantBiasScale_4, radius=@MaterialExpressionConstant_15) {
   code="float3 k = float3(-0.866025404,0.5,0.577350269);\r\nuv = abs(uv);\r\nuv -= 2.0*min(dot(k.xy,uv),0.0)*k.xy;\r\nuv -= float2(clamp(uv.x, -k.z*radius, k.z*radius), radius);\r\nreturn length(uv)*sign(uv.y);"
   description=HexagonSDF
   inputs [
    {inputName=uv, input {expression=&self:MaterialExpressionConstantBiasScale_4}}
    {inputName=radius, input {expression=&self:MaterialExpressionConstant_15}}
   ]
  }
  n11 = Frac(None=Multiply(A=Time(), B=Constant() {r=0.02}))
  n12 = Frac(None=Multiply(A=Time(), B=Constant() {desc=RippleFrequency, r=0.02}))
  n13 = Multiply(A=Subtract(A=@n10, B=Add(A=@n11, B=@n12)), B=Constant() {desc=RippleFrequency, r=3})
  Cosine(None=@n13) {period=3.2}
  n14 = MaterialFunctionCall(X=Constant() {r=-1.2}, Y=Constant() {r=-1.2}) {materialFunction=&MakeFloat2}
  n15 = MaterialFunctionCall() {materialFunction=&MF_UI_UVScaling}
  MaterialExpressionConstantBiasScale_6 = ConstantBiasScale(None=MaterialFunctionCall(UVs=Add(A=@n14.Result, B=@n15.Scaled To Fill), Texture Scale=Constant() {r=0.8}) {materialFunction=&ScaleUVsByCenter}.UVs)
  MaterialExpressionConstant_73 = Constant() {r=1}
  n16 = Custom(uv=@MaterialExpressionConstantBiasScale_6, radius=@MaterialExpressionConstant_73) {
   code="float3 k = float3(-0.866025404,0.5,0.577350269);\r\nuv = abs(uv);\r\nuv -= 2.0*min(dot(k.xy,uv),0.0)*k.xy;\r\nuv -= float2(clamp(uv.x, -k.z*radius, k.z*radius), radius);\r\nreturn length(uv)*sign(uv.y);"
   description=HexagonSDF
   inputs [
    {inputName=uv, input {expression=&self:MaterialExpressionConstantBiasScale_6}}
    {inputName=radius, input {expression=&self:MaterialExpressionConstant_73}}
   ]
  }
  n17 = Multiply(A=Subtract(A=@n16, B=@n6), B=Constant() {desc=RippleFrequency, r=3})
  Cosine(None=@n17) {period=3.2}
  Divide() {constB=1}
  n18 = Constant() {r=3}
  Divide(A=Round(None=Multiply(B=@n18)), B=@n18)
  Divide() {constB=1}
  Frac(None=@n6)
  Frac(None=@n7)
  Frac(None=@n17)
  Frac()
  n19 = Constant3Vector() {constant=<rgba 1 0.0729 0.2438 1>}
  n20 = Constant3Vector() {constant=<rgba 0.7913 0.003 0.1221 1>}
  n21 = Constant3Vector() {constant=<rgba 0.2874 0.0006 0.0467 1>}
  MaterialExpressionConstantBiasScale_3 = ConstantBiasScale(None=MaterialFunctionCall() {materialFunction=&ScaleUVsByCenter}.UVs) {bias=-0.5, scale=2}
  MaterialExpressionConstant_16 = Constant() {r=1}
  n22 = Custom(uv=@MaterialExpressionConstantBiasScale_3, radius=@MaterialExpressionConstant_16) {
   code="float3 k = float3(-0.866025404,0.5,0.577350269);\r\nuv = abs(uv);\r\nuv -= 2.0*min(dot(k.xy,uv),0.0)*k.xy;\r\nuv -= float2(clamp(uv.x, -k.z*radius, k.z*radius), radius);\r\nreturn length(uv)*sign(uv.y);"
   description=HexagonSDF
   inputs [
    {inputName=uv, input {expression=&self:MaterialExpressionConstantBiasScale_3}}
    {inputName=radius, input {expression=&self:MaterialExpressionConstant_16}}
   ]
  }
  n23 = Time()
  n24 = Constant() {r=0.1}
  n25 = ScalarParameter() {defaultValue=3, parameterName=Bars}
  n26 = Multiply(A=Subtract(A=@n22, B=Multiply(A=@n23, B=@n24)), B=@n25)
  n27 = Frac(None=Divide(A=Round(None=@n26), B=@n25))
  n28 = LinearInterpolate(
   A=Multiply(A=MaterialFunctionCall(A=@n19, B=@n20, C=@n21, Alpha=@n27) {materialFunction=&Lerp_3Color}.Result) {constB=1.5}
   B=Constant3Vector() {constant=<rgba 1 0.0729 0.2438 1>}
   Alpha=MaterialFunctionCall(GradientStart=Constant() {r=0.328}, GradientEnd=Constant() {r=0.9156}, RotationAngle=Constant() {r=0.75}) {materialFunction=&MF_UI_LinearGradient}.Result
  )
  n29 = MaterialFunctionCall() {materialFunction=&MF_UI_UVScaling}
  n30 = Constant() {r=0.02}
  n31 = Constant() {r=0.1}
  n32 = MaterialFunctionCall(
   UV=Add(
    A=MaterialFunctionCall(UVs=@n29.Scaled To Fill) {materialFunction=&ScaleUVsByCenter}.UVs
    B=MaterialFunctionCall(X=@n30, Y=@n31) {materialFunction=&MakeFloat2}.Result
   )
   GradientStart=Constant() {r=0.25}
   GradientEnd=Constant() {r=2.1}
  ) {materialFunction=&MF_UI_RadialGradient}
  n33 = LinearInterpolate(A=Constant3Vector() {constant=<rgba 1 0.0168 0.1981 1>}, Alpha=@n32.Smooth) {constA=0.25, constB=0.75}
  LinearInterpolate(A=@n28, B=@n33, Alpha=@n32.Smooth)
  n34 = Constant()
  n35 = Constant() {r=-0.05}
  n36 = Constant() {r=1.1}
  n37 = TextureSample(UVs=Add(
    A=MaterialFunctionCall(X=@n34, Y=@n35) {materialFunction=&MakeFloat2}.Result
    B=MaterialFunctionCall(UVs=@n29.Scaled To Fill, Texture Scale=@n36) {materialFunction=&ScaleUVsByCenter}.UVs
   )) {texture=&/T_UI_Lettuce_Shard}
  LinearInterpolate(
   A=LinearInterpolate(A=@n28, B=Constant3Vector() {constant=<rgba 0 0.5201 0.8073 1>}, Alpha=@n37.R)
   B=LinearInterpolate(A=@n28, B=Constant3Vector() {constant=<rgba 1 0.975 0.9985 0>}, Alpha=@n37.B)
   Alpha=@n37.B
  )
  n38 = Constant3Vector() {constant=<rgba 1 0.8961 0.8349 0>}
  n39 = Constant() {r=-0.1}
  n40 = Constant() {r=0.25}
  n41 = MaterialFunctionCall(
   UV=Add(
    A=MaterialFunctionCall() {materialFunction=&ScaleUVsByCenter}.UVs
    B=MaterialFunctionCall(X=@n39, Y=@n40) {materialFunction=&MakeFloat2}.Result
   )
   GradientStart=Constant() {r=0.15}
   GradientEnd=Constant() {r=0.8727}
  ) {materialFunction=&MF_UI_RadialGradient}
  n42 = Constant3Vector() {constant=<rgba 1 0.975 0.9985 0>}
  n43 = Constant3Vector() {constant=<rgba 0.0026 0.2848 1 0>}
  n44 = MaterialFunctionCall(GradientStart=Constant() {r=0.6}, GradientEnd=Constant() {r=1.0617}, RotationAngle=Constant() {r=0.75}) {materialFunction=&MF_UI_LinearGradient}
  n45 = Constant() {r=3}
  n46 = Multiply(A=Abs(None=Frac(None=@n13)), B=@n45)
  n47 = MaterialFunctionCall(
   A=Constant3Vector() {constant=<rgba 0.467 0 0.0732 1>}
   B=Constant3Vector() {constant=<rgba 0.8351 0 0.1252 1>}
   C=Constant3Vector() {constant=<rgba 1 0.0482 0.2582 1>}
   Alpha=Divide(A=Round(None=@n46), B=@n45)
  ) {materialFunction=&Lerp_3Color}
  LinearInterpolate(
   A=Multiply(A=LinearInterpolate(B=@n38, Alpha=@n41.Smooth)) {constB=0.184}
   B=Multiply(A=LinearInterpolate(A=@n42, B=@n43, Alpha=@n44.Result), B=Multiply(A=@n47.Result) {constB=1.5})
  )
  MaterialFunctionCall() {materialFunction=&LinearGradient}
  n48 = Constant() {r=2}
  n49 = Multiply(A=Cosine(None=@n5), B=@n48)
  MaterialFunctionCall(
   A=Constant3Vector() {constant=<rgba 0.2789 0 0.0437 1>}
   B=Constant3Vector() {constant=<rgba 0.7991 0.0027 0.1221 1>}
   C=Constant3Vector() {constant=<rgba 1 0.0482 0.2582 1>}
   Alpha=Divide(A=Round(None=@n49), B=@n48)
  ) {materialFunction=&Lerp_3Color}
  Multiply()
  Multiply(A=@n33) {constB=0.256}
  n50 = Time()
  n51 = Constant() {r=1.25}
  n52 = Frac(None=Multiply(A=Add(A=@n50, B=@n51) {constB=0.25}, B=ScalarParameter() {defaultValue=0.15, parameterName=speed}))
  n53 = Constant() {r=1.75}
  MaterialExpressionConstantBiasScale_0 = ConstantBiasScale(None=MaterialFunctionCall(
    UVs=MaterialFunctionCall() {materialFunction=&MF_UI_UVScaling}.Scaled to Fit
    Texture Scale=Multiply(A=@n52, B=@n53)
   ) {materialFunction=&ScaleUVsByCenter}.UVs) {bias=-0.5, scale=2}
  MaterialExpressionConstant_0 = Constant() {r=1}
  n54 = Custom!(p=@MaterialExpressionConstantBiasScale_0, r=@MaterialExpressionConstant_0) {
   code="const float3 k = float3(-0.866025404,0.5,0.577350269);\r\n    p = abs(p);\r\n    p -= 2.0*min(dot(k.xy,p),0.0)*k.xy;\r\n    p -= float2(clamp(p.x, -k.z*r, k.z*r), r);\r\n    return length(p)*sign(p.y);"
   outputType=CMOT_Float1
   inputs [
    {inputName=p, input {expression=&self:MaterialExpressionConstantBiasScale_0}}
    {inputName=r, input {expression=&self:MaterialExpressionConstant_0}}
   ]
  }
  NamedRerouteDeclaration(None=OneMinus(None=SmoothStep(Value=@n54) {constMax=0.003})) {name=Hexagon_2, nodeColor=<rgba 1 0 0.1021 1>}
  NamedRerouteDeclaration(None=@n4) {name=Hexagon_3, nodeColor=<rgba 0.4212 0 0.9236 1>}
  n55 = Frac(None=Multiply(A=Time(), B=ScalarParameter() {defaultValue=0.15, parameterName=speedHex1}))
  n56 = Constant() {r=1.5}
  MaterialExpressionConstantBiasScale_2 = ConstantBiasScale(None=MaterialFunctionCall(
    UVs=MaterialFunctionCall() {materialFunction=&MF_UI_UVScaling}.Scaled to Fit
    Texture Scale=Multiply(A=@n55, B=@n56)
   ) {materialFunction=&ScaleUVsByCenter}.UVs) {bias=-0.5, scale=2}
  MaterialExpressionConstant_9 = Constant() {r=1}
  n57 = Custom(p=@MaterialExpressionConstantBiasScale_2, r=@MaterialExpressionConstant_9) {
   code="const float3 k = float3(-0.866025404,0.5,0.577350269);\r\n    p = abs(p);\r\n    p -= 2.0*min(dot(k.xy,p),0.0)*k.xy;\r\n    p -= float2(clamp(p.x, -k.z*r, k.z*r), r);\r\n    return length(p)*sign(p.y);"
   outputType=CMOT_Float1
   inputs [
    {inputName=p, input {expression=&self:MaterialExpressionConstantBiasScale_2}}
    {inputName=r, input {expression=&self:MaterialExpressionConstant_9}}
   ]
  }
  NamedRerouteDeclaration(None=OneMinus(None=SmoothStep(Value=@n57) {constMax=0.003})) {name=Hexagon, nodeColor=<rgba 0 0.9614 1 1>}
  OneMinus()
  OneMinus(None=@n4)
  MaterialExpressionConstantBiasScale_5 = ConstantBiasScale(None=MaterialFunctionCall() {materialFunction=&ScaleUVsByCenter}.UVs) {bias=-0.5, scale=2}
  MaterialExpressionConstant_28 = Constant() {r=1}
  n58 = Custom!(uv=@MaterialExpressionConstantBiasScale_5, radius=@MaterialExpressionConstant_28) {
   code="float3 k = float3(-0.866025404,0.5,0.577350269);\r\nuv = abs(uv);\r\nuv -= 2.0*min(dot(k.xy,uv),0.0)*k.xy;\r\nuv -= float2(clamp(uv.x, -k.z*radius, k.z*radius), radius);\r\nreturn length(uv)*sign(uv.y);"
   description=HexagonSDF
   inputs [
    {inputName=uv, input {expression=&self:MaterialExpressionConstantBiasScale_5}}
    {inputName=radius, input {expression=&self:MaterialExpressionConstant_28}}
   ]
  }
  n59 = MaterialFunctionCall(Multiplier=ScalarParameter() {defaultValue=3, parameterName=Bars}) {materialFunction=&Pi}
  n60 = Time()
  n61 = Constant() {r=0.1}
  Sine(None=Subtract(A=Multiply(A=@n58, B=@n59.Result), B=Multiply(A=@n60, B=@n61)))
  SmoothStep(Value=@n4) {constMax=0.003}
  Subtract(A=@n4)
  TextureCoordinate()
  TextureCoordinate()
  TextureCoordinate()
  TextureSample(UVs=MaterialFunctionCall(
    UVs=MaterialFunctionCall() {materialFunction=&MF_UI_UVScaling}.Scaled To Fill
    Texture Scale=Constant() {r=1.1}
   ) {materialFunction=&ScaleUVsByCenter}.UVs) {texture=/OfferCatalog/Art/Lettuce/T_UI_LettuceBG}
 ]
}
